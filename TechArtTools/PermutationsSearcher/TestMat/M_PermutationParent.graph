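# M_PermutationParent is a parent test material in the PermutationsSearcher tech art tool folder, apparently for exercising static switch permutations.
# Its Default Lit graph holds two static switch parameters, DepthFade and CameraFade, whose outputs feed an Add node.
# All switch inputs are default constants and the Add drives no material output, so the switches perform no actual fading.
# Automatic usage flag setting in the editor is disabled.
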Material /TechArtTools/PermutationsSearcher/TestMat/M_PermutationParent
graph {
 properties {bAutomaticallySetUsageInEditor=false, usedShadingModels=Default Lit}
 nodes [
  n1 = Constant()
  Add(
   A=StaticSwitchParameter(True=Constant(), False=@n1) {parameterName=DepthFade}
   B=StaticSwitchParameter(True=@n1, False=@n1) {parameterName=CameraFade}
  )
 ]
}
